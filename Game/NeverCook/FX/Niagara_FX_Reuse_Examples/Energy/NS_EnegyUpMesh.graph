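# Niagara system from an FX reuse examples folder that shows an energy-up mesh effect through a single emitter, UpMesh.
# After a short loop delay, an instantaneous burst spawns one mesh particle, stretched along Z and given a material random value.
# An alpha curve fades it in, holds it, then fades it out over its lifetime.
# No module adds velocity or force, and no user parameters are exposed.
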NiagaraSystem /Game/NeverCook/FX/Niagara_FX_Reuse_Examples/Energy/NS_EnegyUpMesh
dependencies {usedDataInterfaces [&NiagaraDataInterfaceCurve], usedDynamicInputs [], usedModules [&SystemState]}
emitters {UpMesh {
  emitterUpdateScript [
   &EmitterState(
    Inactive Response=::NewEnumerator1
    Life Cycle Mode=::NewEnumerator1
    Loop Behavior=::NewEnumerator1
    Loop Delay=0.2
    Loop Duration=2
    UseLoopDelay=true
   )
   &SpawnBurst_Instantaneous(Spawn Count=1, Use Spawn Probability=false)
  ]
  particleSpawnScript [
   &InitializeParticle(Lifetime=3, Mesh Scale=<1 1 3>, Mesh Scale Mode=::NewEnumerator4, UsePositionOffset=false, Write Material Random=true) {module=InitializeParticle001}
   &InitialMeshOrientation(Mesh Orientation Mode=::NewEnumerator2, Use Rotation Vector=false)
  ]
  particleUpdateScript [
   &ParticleState()
   &SolveForcesAndVelocity(Clamp Velocity=false, Limit Acceleration=false)
   &ScaleColor(
    Scale Alpha=&FloatFromCurve(FloatCurve=DataInterface() {propertyValues {Curve {keys [<at 0 0>,<at 0.2 1>,<at 0.8 1>,<at 1 0>]}}})
    ScaleA=true
    ScaleRGB=true
   )
  ]
  rendererClasses [&NiagaraMeshRendererProperties]
  renderers [{rendererClass=&NiagaraMeshRendererProperties, rendererIndex=0}]
 }}
summary {
 systemName=NS_EnegyUpMesh
 emitters [{emitterName=UpMesh, rendererClasses [&NiagaraMeshRendererProperties]}]
 userVariables []
}
