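# Niagara reuse-example energy ball effect, built from a single mesh emitter named Center001.
# An instantaneous burst spawns oriented mesh particles that outlive their lifetime and can die when the emitter deactivates.
# Each update tints color pink-magenta with boosted alpha and writes Dark_Color (deep purple) and Dark_Size (enlarged particle scale).
# Two mesh renderers draw the particles; the Dark_ attributes suggest a layered bright core and darker shell, though bindings are not shown.
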NiagaraSystem /Game/NeverCook/FX/Niagara_FX_Reuse_Examples/Energy/NS_Energy_Ball
dependencies {usedDataInterfaces [], usedDynamicInputs [], usedModules [&SystemState]}
emitters {Center001 {
  emitterUpdateScript [
   &EmitterState(Life Cycle Mode=::NewEnumerator1, Loop Behavior=::NewEnumerator1, Loop Duration Mode=::NewEnumerator1, UseLoopDelay=false)
   &SpawnBurst_Instantaneous(Use Spawn Probability=false, UseLoopCountLimit=false)
  ]
  particleSpawnScript [
   &InitializeParticle(Color Mode=::NewEnumerator0, Mesh Scale Mode=::NewEnumerator3, UsePositionOffset=false, Write Material Random=true)
   &InitialMeshOrientation(Mesh Orientation Mode=::NewEnumerator2, Rotation=<0 0.25 0>, Use Rotation Vector=true)
  ]
  particleUpdateScript [
   &ParticleState(Kill Particles When Lifetime Has Elapsed=false, Let Infinitely Lived Particles Die When Emitter Deactivates=true)
   &ScaleColor(
    Scale Alpha=1.5
    Scale RGB=&MakeVectorFromLinearColorRGB(Color=<rgba 1 0.151 0.9716 1>)
    ScaleA=true
    ScaleRGB=true
   )
   /SetVariables(Particles.Dark_Color:LinearColor=&MakeLinearColorFromVectorAndFloat("Vector (RGB)"=&MakeVectorFromLinearColorRGB(Color=<rgba 0.0744 0.0078 0.1562 1>), "Float (Alpha)"=1)) {isSetParameters=true}
   /SetVariables(Particles.Dark_Size:Vector3f=&Multiply_VectorByFloat(Vector=$Particles.Scale, Float=1.4)) {isSetParameters=true}
  ]
  rendererClasses [&NiagaraMeshRendererProperties]
  renderers [
   {rendererClass=&NiagaraMeshRendererProperties, rendererIndex=0}
   {rendererClass=&NiagaraMeshRendererProperties, rendererIndex=1}
  ]
 }}
summary {
 systemName=NS_Energy_Ball
 emitters [{emitterName=Center001, rendererClasses [&NiagaraMeshRendererProperties]}]
 userVariables []
}
