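# NS_Energy_Wave is a Niagara example effect for an energy wave, built from a single mesh emitter named Absorb_Energy.
# An instantaneous burst spawns short-lived meshes with a random rotation on one axis.
# Over each lifetime, curves grow the mesh from partial to full scale and fade alpha in and out.
# Dynamic material parameters and per-particle material random feed the mesh material, and no user parameters are exposed.
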NiagaraSystem /Game/NeverCook/FX/Niagara_FX_Reuse_Examples/Energy/NS_Energy_Wave
dependencies {usedDataInterfaces [&NiagaraDataInterfaceCurve], usedDynamicInputs [], usedModules [&SystemState]}
emitters {Absorb_Energy {
  emitterUpdateScript [
   &EmitterState(Inactive Response=::NewEnumerator1, Life Cycle Mode=::NewEnumerator1, Loop Behavior=::NewEnumerator1, Loop Duration=1.5, UseLoopDelay=false)
   &SpawnBurst_Instantaneous(Use Spawn Probability=false, UseLoopCountLimit=false)
  ]
  particleSpawnScript [
   &InitializeParticle(Lifetime=1.2, Mesh Scale Mode=::NewEnumerator3, Mesh Uniform Scale=1.5, UsePositionOffset=false, Write Material Random=true)
   &InitialMeshOrientation(
    Mesh Orientation Mode=::NewEnumerator2
    Rotation=&RandomRangeVector(Minimum=<0 0 0>, Maximum=<1 0 0>)
    Use Rotation Vector=true
   )
  ]
  particleUpdateScript [
   &ParticleState()
   &ScaleColor(
    Scale Alpha=&FloatFromCurve(FloatCurve=DataInterface() {propertyValues {Curve {keys [
         {arriveTangent=0.0004, interpMode=RCIM_Cubic, leaveTangent=0.0004, tangentMode=RCTM_User, time=0, value=0}
         {arriveTangent=0.02, interpMode=RCIM_Cubic, leaveTangent=0.02, tangentMode=RCTM_User, time=0.1, value=1}
         {arriveTangent=-0.0087, interpMode=RCIM_Cubic, leaveTangent=-0.0087, tangentMode=RCTM_User, time=0.8, value=1}
         {arriveTangent=0.0459, interpMode=RCIM_Cubic, leaveTangent=0.0459, tangentMode=RCTM_User, time=1, value=0}
        ]}}})
    Scale RGB=&MakeVectorFromLinearColorRGB(Color=<rgba 1 1 1 1>)
    ScaleA=true
    ScaleRGB=true
   )
   &DynamicMaterialParameters(
    Index 0 Param 3=-1.5
    Module.Index0Float0Write=true
    Module.Index0Float1Write=true
    Module.Index0Float2Write=true
    Module.Index0Float3Write=true
   )
   &ScaleMeshSize(Scale Factor=&VectorFromFloat(Value=&FloatFromCurve(FloatCurve=DataInterface() {propertyValues {Curve {keys [{arriveTangent=0.4936, interpMode=RCIM_Cubic, leaveTangent=0.4936, tangentMode=RCTM_User, time=0, value=0.4}, <at 1 1 RCIM_Cubic>]}}})))
  ]
  rendererClasses [&NiagaraMeshRendererProperties]
  renderers [{rendererClass=&NiagaraMeshRendererProperties, rendererIndex=0}]
 }}
summary {
 systemName=NS_Energy_Wave
 emitters [{emitterName=Absorb_Energy, rendererClasses [&NiagaraMeshRendererProperties]}]
 userVariables []
}
