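# NS_FloatEnergy is a Niagara energy effect, in an FX reuse examples folder, that shows one short-lived flare sprite.
# Its only emitter, Flash_Flare, burst-spawns a single tall sprite after a brief delay.
# Curves fade its alpha in and out and drive a falling dynamic material parameter.
# It also computes light color and radius with a randomized flash curve, but only a sprite renderer is listed.
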NiagaraSystem /Game/NeverCook/FX/Niagara_FX_Reuse_Examples/Energy/NS_FloatEnergy
dependencies {usedDataInterfaces [&NiagaraDataInterfaceCurve], usedDynamicInputs [], usedModules [&SystemState]}
emitters {Flash_Flare {
  emitterUpdateScript [
   &EmitterState(Life Cycle Mode=::NewEnumerator1, Loop Behavior=::NewEnumerator1, Loop Duration=1, UseLoopDelay=false)
   &SpawnBurst_Instantaneous(Spawn Count=1, Spawn Time=0.1, Use Spawn Probability=false)
  ]
  particleSpawnScript [&InitializeParticle(
    Color=<rgba 10 10 10 1>
    Lifetime=1.5
    Sprite Size=<50 200>
    Sprite Size Mode=::NewEnumerator4
    UsePositionOffset=false
    Write Material Random=true
   )]
  particleUpdateScript [
   &ParticleState()
   &ScaleSpriteSize(Scale Sprite Size Mode=::NewEnumerator0, Uniform Scale Factor=3)
   &ScaleColor(
    Scale Alpha=&FloatFromCurve(FloatCurve=DataInterface() {propertyValues {Curve {keys [
         {arriveTangent=-0.0383, interpMode=RCIM_Cubic, leaveTangent=-0.0383, tangentMode=RCTM_User, time=0, value=0}
         {arriveTangent=-0.0405, interpMode=RCIM_Cubic, leaveTangent=-0.0405, tangentMode=RCTM_User, time=0.2, value=0.95}
         {arriveTangent=-1.1875, interpMode=RCIM_Cubic, leaveTangent=-1.1875, time=0.709, value=0.8634}
         {arriveTangent=0.0036, interpMode=RCIM_Cubic, leaveTangent=0.0036, tangentMode=RCTM_User, time=1, value=0}
        ]}}})
    Scale RGB=&VectorFromFloat(Value=0.2)
    ScaleA=true
    ScaleRGB=true
   )
   &FN_Light_Attributes(
    Color=$Particles.Color
    Create Color Attribute=true
    Create Radius Attribute=true
    Radius=600
    Scale Color=&FloatFromCurve(
     FloatCurve=DataInterface() {propertyValues {Curve {keys [
         <at 0 0 RCIM_Cubic>
         <at 0.1 0.95 RCIM_Cubic>
         {arriveTangent=-2.6662, interpMode=RCIM_Cubic, leaveTangent=-2.6662, tangentMode=RCTM_User, time=1, value=0}
        ]}}}
     Scale Curve=&RandomRangeFloat(Minimum=600, Maximum=800, Evaluation Type=::NewEnumerator1)
    )
   )
   &DynamicMaterialParameters(
    Index 0 Param 1=&FloatFromCurve(FloatCurve=DataInterface() {propertyValues {Curve {keys [<at 0 1>,<at 1 0>]}}}, Scale Curve=3)
    Module.Index0Float0Write=true
    Module.Index0Float1Write=true
    Module.Index0Float2Write=true
    Module.Index0Float3Write=true
   )
  ]
  rendererClasses [&NiagaraSpriteRendererProperties]
  renderers [{rendererClass=&NiagaraSpriteRendererProperties}]
 }}
summary {
 systemName=NS_FloatEnergy
 emitters [{emitterName=Flash_Flare, rendererClasses [&NiagaraSpriteRendererProperties]}]
 userVariables []
}
