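# NS_SplineThicken_energy is a Niagara energy effect from a reuse-examples folder, built from one mesh emitter, Spiral_AirFlow.
# An instantaneous burst spawns meshes with a fixed initial rotation and a camera offset.
# Updates apply drag, stretch each mesh along X, and scale color by a dimming gray curve with random brightness, while alpha rises quickly then fades out.
# The system exposes no user parameters.
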NiagaraSystem /Game/NeverCook/FX/Niagara_FX_Reuse_Examples/Energy/NS_SplineThicken_energy
dependencies {
 usedDataInterfaces [&NiagaraDataInterfaceColorCurve, &NiagaraDataInterfaceCurve]
 usedDynamicInputs []
 usedModules [&SystemState]
}
emitters {Spiral_AirFlow {
  emitterUpdateScript [
   &EmitterState(Life Cycle Mode=::NewEnumerator1, Loop Behavior=::NewEnumerator1, Loop Delay=0.03, Loop Duration=1, UseLoopDelay=false)
   &SpawnBurst_Instantaneous(Use Spawn Probability=false, UseLoopCountLimit=false)
  ]
  particleSpawnScript [
   &InitializeParticle(
    Color=<rgba 2 2 2 2>
    Lifetime Min=2
    Lifetime Mode=::NewEnumerator1
    Mesh Scale Mode=::NewEnumerator1
    Mesh Uniform Scale Max=1
    Position Offset=<0 0 200>
    UsePositionOffset=false
    Write Material Random=true
   )
   &CameraOffset(Camera Offset Amount=10)
   &InitialMeshOrientation(Mesh Orientation Mode=::NewEnumerator2, Rotation=<0.15 0 0>, Use Rotation Vector=true)
  ]
  particleUpdateScript [
   &Drag(Drag=2, Use Linear Drag=true, Use Rotational Drag=true)
   &SolveForcesAndVelocity(Clamp Velocity=false, Limit Acceleration=false)
   &ParticleState()
   &ScaleMeshSize(Scale Factor=<2.5 1 1>) {module=ScaleMeshSize001}
   &ScaleColor(
    Scale Alpha=&FloatFromCurve(
     FloatCurve=DataInterface() {propertyValues {Curve {keys [
         {interpMode=RCIM_Cubic, tangentMode=RCTM_User, time=0, value=0}
         {arriveTangent=-1.3837, interpMode=RCIM_Cubic, leaveTangent=-1.3837, tangentMode=RCTM_User, time=0.2, value=0.8}
         {arriveTangent=-3.6419, tangentMode=RCTM_User, time=1, value=0}
        ]}}}
     Scale Curve=&Multiply_Float(A=0.5, B=1)
    )
    Scale RGB=&Multiply_VectorByFloat(
     Vector=&MakeVectorFromLinearColorRGB(Color=&ColorFromCurve(ColorCurve=DataInterface() {propertyValues {
         AlphaCurve {keys []}
         BlueCurve {keys [<at 0 0.7 RCIM_Cubic>,<at 0.5852 0.5 RCIM_Cubic>]}
         GreenCurve {keys [<at 0 0.7 RCIM_Cubic>,<at 0.5852 0.5 RCIM_Cubic>]}
         RedCurve {keys [<at 0 0.7 RCIM_Cubic>,<at 0.5852 0.5 RCIM_Cubic>]}
        }}))
     Float=&RandomRangeFloat(Minimum=0.5, Maximum=1)
    )
    ScaleA=true
    ScaleRGB=true
   )
  ]
  rendererClasses [&NiagaraMeshRendererProperties]
  renderers [{rendererClass=&NiagaraMeshRendererProperties, rendererIndex=0}]
 }}
summary {
 systemName=NS_SplineThicken_energy
 emitters [{emitterName=Spiral_AirFlow, rendererClasses [&NiagaraMeshRendererProperties]}]
 userVariables []
}
